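# Base stereo one-shot MetaSound source for simple cosmetic sounds, exposing a wave array, volume and pitch ranges, start delay and template multipliers.
# On play, a random time between the StartDelay bounds delays the trigger into MSS_PlayRandomSound_Stereo, which receives the waves and modulation ranges.
# TemplatePitchMultiplier scales both pitch bounds beforehand, while TemplateVolumeMultiplier sets the gain of a stereo mixer on the output.
# The nested player's finished trigger ends the source.
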MetaSoundSource /FortMetasound/Cosmetics/MSS_Cosmetic_Simple_Base
inputs [
 PitchModulationRange.Max {type=Float}
 PitchModulationRange.Min {type=Float}
 StartDelay.Max {type=Time}
 StartDelay.Min {type=Time}
 TemplatePitchMultiplier {type=Float}
 TemplateVolumeMultiplier {type=Float}
 UE.Source.OnPlay {type=Trigger}
 VolumeModulationRange.Max {type=Float}
 VolumeModulationRange.Min {type=Float}
 Waves {type=WaveAsset:Array}
]
interfaces [UE.OutputFormat.Stereo:1.0, UE.Source.OneShot:1.0, UE.Source:1.1]
outputs [
 "UE.OutputFormat.Stereo.Audio:0" {type=Audio}
 "UE.OutputFormat.Stereo.Audio:1" {type=Audio}
 UE.Source.OneShot.OnFinished {type=Trigger}
]
pages [Default {
  input_defaults {
   PitchModulationRange.Max=1
   PitchModulationRange.Min=1
   StartDelay.Max=0
   StartDelay.Min=0
   TemplatePitchMultiplier=1
   TemplateVolumeMultiplier=1
   UE.Source.OnPlay=null
   VolumeModulationRange.Max=1
   VolumeModulationRange.Min=1
   Waves []
  }
  nodes [
   n1 = UE.RandomTime(Max=^StartDelay.Max, Min=^StartDelay.Min, Next=^UE.Source.OnPlay)
   TriggerDelay_0 = UE.Trigger Delay(In=@n1.On Next, Delay Time=@n1.Value)
   MSS_PlayRandomSound_Stereo_0 = &MSS_PlayRandomSound_Stereo(
    MaxVolume=^VolumeModulationRange.Max
    MinVolume=^VolumeModulationRange.Min
    Sounds=^Waves
    MaxPitch=UE.Multiply:Float(PrimaryOperand=^PitchModulationRange.Max, AdditionalOperands=^TemplatePitchMultiplier).out
    MinPitch=UE.Multiply:Float(PrimaryOperand=^PitchModulationRange.Min, AdditionalOperands=^TemplatePitchMultiplier).out
    UE.Source.OnPlay=@TriggerDelay_0.out
   )
   AudioMixerAudioMixerStereo2_0 = "AudioMixer.Audio Mixer (Stereo, 2)"(
    Gain 0=^TemplateVolumeMultiplier
    In 0 L=@MSS_PlayRandomSound_Stereo_0."UE.OutputFormat.Stereo.Audio:0"
    In 0 R=@MSS_PlayRandomSound_Stereo_0."UE.OutputFormat.Stereo.Audio:1"
   )
  ]
  links [
   {dst="@UE.OutputFormat.Stereo.Audio:0", src=AudioMixerAudioMixerStereo2_0.Out L}
   {dst="@UE.OutputFormat.Stereo.Audio:1", src=AudioMixerAudioMixerStereo2_0.Out R}
   {dst="@UE.Source.OneShot.OnFinished", src=MSS_PlayRandomSound_Stereo_0.UE.Source.OneShot.OnFinished}
  ]
 }]
properties {assetImportData=&self:AssetImportData, numChannels=2, outputFormat=Stereo}
